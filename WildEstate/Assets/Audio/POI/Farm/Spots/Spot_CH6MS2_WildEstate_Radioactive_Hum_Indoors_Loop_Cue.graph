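# Looping indoor radioactive hum ambience cue for the WildEstate farm POI, attenuated by a radioactive box setting and sent to reverb.
# A quality switch feeds its first two slots a random pick of four stereo loops and its last two the fourth loop alone.
# Each branch adds a short random start delay, modulation, and two stacked endlessly looping 60-second envelopers with pitch and wavering volume curves.
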SoundCue /WildEstate/Assets/Audio/POI/Farm/Spots/Spot_CH6MS2_WildEstate_Radioactive_Hum_Indoors_Loop_Cue
root=QualityLevel_0
nodes [
 {id=Delay_0, children [WavePlayer_10], properties {delayMax=0.2}}
 {id=Delay_1, children [Random_1], properties {delayMax=0.2}}
 {
  id=Enveloper_0
  children [Modulator_0]
  properties {
   bLoop=true
   bLoopIndefinitely=true
   loopEnd=60
   pitchCurve {editorCurveData {keys [{interpMode=RCIM_Cubic, time=0}, <at 30 0.99 RCIM_Cubic>, {time=60}]}}
   volumeCurve {editorCurveData {keys [<at 0 1>]}}
  }
 }
 {
  id=Enveloper_1
  children [Enveloper_0]
  properties {
   bLoop=true
   bLoopIndefinitely=true
   loopEnd=60
   pitchCurve {editorCurveData {keys [{time=0}]}}
   volumeCurve {editorCurveData {keys [
      {arriveTangent=-0.01, interpMode=RCIM_Cubic, time=0, value=1}
      {arriveTangent=-0.0035, interpMode=RCIM_Cubic, leaveTangent=-0.0035, tangentMode=RCTM_User, time=4.5428, value=0.8801}
      {arriveTangent=0.0016, interpMode=RCIM_Cubic, leaveTangent=0.0016, tangentMode=RCTM_User, time=7.4135, value=0.9776}
      {arriveTangent=0.0008, interpMode=RCIM_Cubic, leaveTangent=0.0008, tangentMode=RCTM_User, time=11.1077, value=0.8936}
      {arriveTangent=0.0016, interpMode=RCIM_Cubic, leaveTangent=0.0016, tangentMode=RCTM_User, time=12.9775, value=0.8792}
      {arriveTangent=0.0009, interpMode=RCIM_Cubic, leaveTangent=0.0009, time=14.965, value=0.9463}
      {arriveTangent=-0.0011, interpMode=RCIM_Cubic, leaveTangent=-0.0011, tangentMode=RCTM_User, time=18.5868, value=0.8842}
      {arriveTangent=0.0002, interpMode=RCIM_Cubic, leaveTangent=0.0002, time=23.9559, value=0.9962}
      {arriveTangent=0.0016, interpMode=RCIM_Cubic, leaveTangent=0.0016, tangentMode=RCTM_User, time=28.034, value=0.8864}
      {arriveTangent=-0.0003, interpMode=RCIM_Cubic, leaveTangent=-0.0003, time=31.2025, value=0.961}
      {arriveTangent=0.0009, interpMode=RCIM_Cubic, leaveTangent=0.0009, tangentMode=RCTM_User, time=34.9227, value=0.8844}
      {arriveTangent=-0.0008, interpMode=RCIM_Cubic, leaveTangent=-0.0008, time=37.93, value=0.9962}
      <at 42.0081 0.879>
      {arriveTangent=0.0024, interpMode=RCIM_Cubic, leaveTangent=0.0024, tangentMode=RCTM_User, time=44.5668, value=0.8842}
      {arriveTangent=-0.0006, interpMode=RCIM_Cubic, leaveTangent=-0.0006, time=47.958, value=0.9776}
      {arriveTangent=-0.0056, interpMode=RCIM_Cubic, leaveTangent=-0.0056, tangentMode=RCTM_User, time=53.1343, value=0.8792}
      <at 55.3163 0.9711>
      {arriveTangent=-0.0003, interpMode=RCIM_Cubic, leaveTangent=-0.0003, tangentMode=RCTM_User, time=57.7536, value=0.8789}
      <at 60 1 RCIM_Cubic>
     ]}}
  }
 }
 {
  id=Enveloper_5
  children [Enveloper_6]
  properties {
   bLoop=true
   bLoopIndefinitely=true
   loopEnd=60
   pitchCurve {editorCurveData {keys [{time=0}]}}
   volumeCurve {editorCurveData {keys [
      {arriveTangent=-0.01, interpMode=RCIM_Cubic, time=0, value=1}
      {arriveTangent=-0.0035, interpMode=RCIM_Cubic, leaveTangent=-0.0035, tangentMode=RCTM_User, time=4.5428, value=0.8801}
      {arriveTangent=0.0016, interpMode=RCIM_Cubic, leaveTangent=0.0016, tangentMode=RCTM_User, time=7.4135, value=0.9776}
      {arriveTangent=0.0008, interpMode=RCIM_Cubic, leaveTangent=0.0008, tangentMode=RCTM_User, time=11.1077, value=0.8936}
      {arriveTangent=0.0016, interpMode=RCIM_Cubic, leaveTangent=0.0016, tangentMode=RCTM_User, time=12.9775, value=0.8792}
      {arriveTangent=0.0009, interpMode=RCIM_Cubic, leaveTangent=0.0009, time=14.965, value=0.9463}
      {arriveTangent=-0.0011, interpMode=RCIM_Cubic, leaveTangent=-0.0011, tangentMode=RCTM_User, time=18.5868, value=0.8842}
      {arriveTangent=0.0002, interpMode=RCIM_Cubic, leaveTangent=0.0002, time=23.9559, value=0.9962}
      {arriveTangent=0.0016, interpMode=RCIM_Cubic, leaveTangent=0.0016, tangentMode=RCTM_User, time=28.034, value=0.8864}
      {arriveTangent=-0.0003, interpMode=RCIM_Cubic, leaveTangent=-0.0003, time=31.2025, value=0.961}
      {arriveTangent=0.0009, interpMode=RCIM_Cubic, leaveTangent=0.0009, tangentMode=RCTM_User, time=34.9227, value=0.8844}
      {arriveTangent=-0.0008, interpMode=RCIM_Cubic, leaveTangent=-0.0008, time=37.93, value=0.9962}
      <at 42.0081 0.879>
      {arriveTangent=0.0024, interpMode=RCIM_Cubic, leaveTangent=0.0024, tangentMode=RCTM_User, time=44.5668, value=0.8842}
      {arriveTangent=-0.0006, interpMode=RCIM_Cubic, leaveTangent=-0.0006, time=47.958, value=0.9776}
      {arriveTangent=-0.0056, interpMode=RCIM_Cubic, leaveTangent=-0.0056, tangentMode=RCTM_User, time=53.1343, value=0.8792}
      <at 55.3163 0.9711>
      {arriveTangent=-0.0003, interpMode=RCIM_Cubic, leaveTangent=-0.0003, tangentMode=RCTM_User, time=57.7536, value=0.8789}
      <at 60 1 RCIM_Cubic>
     ]}}
  }
 }
 {
  id=Enveloper_6
  children [Modulator_1]
  properties {
   bLoop=true
   bLoopIndefinitely=true
   loopEnd=60
   pitchCurve {editorCurveData {keys [{interpMode=RCIM_Cubic, time=0}, <at 30 0.99 RCIM_Cubic>, {time=60}]}}
   volumeCurve {editorCurveData {keys [<at 0 1>]}}
  }
 }
 {id=Modulator_0, children [Delay_1], properties {pitchMax=1.01, pitchMin=0.99}}
 {id=Modulator_1, children [Delay_0], properties {volumeMax=1, volumeMin=1}}
 {id=QualityLevel_0, children [Enveloper_1,Enveloper_1,Enveloper_5,Enveloper_5]}
 {id=Random_1, children [WavePlayer_3,WavePlayer_7,WavePlayer_8,WavePlayer_9], properties {weights [1,1,1,1]}}
 {
  id=WavePlayer_10
  properties {bLooping=true, soundWaveAssetPtr=&/Spot_WildEstate_Radioactive_04_Loop_Stereo}
 }
 {
  id=WavePlayer_3
  properties {bLooping=true, soundWaveAssetPtr=&/Spot_WildEstate_Radioactive_01_Loop_Stereo}
 }
 {
  id=WavePlayer_7
  properties {bLooping=true, soundWaveAssetPtr=&/Spot_WildEstate_Radioactive_02_Loop_Stereo}
 }
 {
  id=WavePlayer_8
  properties {bLooping=true, soundWaveAssetPtr=&/Spot_WildEstate_Radioactive_03_Loop_Stereo}
 }
 {
  id=WavePlayer_9
  properties {bLooping=true, soundWaveAssetPtr=&Spot_WildEstate_Radioactive_04_Loop_Stereo}
 }
]
properties {
 attenuationSettings=&/Fort_Box_Radioactive_Att
 maxDistance=7400
 soundClassObject=&Fort_Ambient_Indoors_Athena
 attenuationOverrides {falloffDistance=2000, nonSpatializedRadiusStart=700, stereoSpread=1000, attenuationShapeExtents {x=925}}
 concurrencyOverrides {maxCount=2, voiceStealReleaseTime=2}
 soundSubmixSends [{maxSendLevel=1.2, minSendLevel=1, sendLevel=1.2, soundSubmix=&Fortnite_Reverb_Default}]
}
